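; BP_TRV_CAS_BigArmchair is a searchable big armchair prop in the Transylvania building set, apparently a loot container.
; Its OnBeginSearch custom event plays a couch search sound cue at the actor's location.
; Its OnLootRepeat custom event plays a "search done" cue there, signalling that loot is not available.
; The construction script is empty, and the graph holds only this audio feedback.
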(event Custom|OnBeginSearch
  ;[1 1 1] Audio
  (Audio|PlaySoundatLocation "/Game/Sounds/Foley_Loot/Containers/Sofa/Prop_Couch_Search_Cue.Prop_Couch_Search_Cue" (Transformation|GetActorLocation)) ; Search Sound
)

(event Custom|OnLootRepeat
  ;[1 1 1] Audio
  (Audio|PlaySoundatLocation "/Game/Sounds/Foley_Loot/Containers/Sofa/Prop_Couch_SearchDone_Cue.Prop_Couch_SearchDone_Cue" (Transformation|GetActorLocation)) ; Loot not available sound
)

(fn ConstructionScript)
