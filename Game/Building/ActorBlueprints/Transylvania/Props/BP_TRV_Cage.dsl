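; BP_TRV_Cage is a Transylvania cage prop Blueprint that adds audio feedback for searching and looting.
; OnBeginSearch plays the cemetery coffin search cue at the actor's location, and OnLootRepeat plays the matching coffin search-end cue there.
; The cage reuses coffin container sounds, and its construction script is empty.
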(event Custom|OnBeginSearch
  ;[1 1 1] Audio
  (Audio|PlaySoundatLocation "/Game/Sounds/Foley_Loot/Containers/Cemetery/Coffin_Search_Cue.Coffin_Search_Cue" (Transformation|GetActorLocation)))

(event Custom|OnLootRepeat
  ;[1 1 1] Audio
  (Audio|PlaySoundatLocation "/Game/Sounds/Foley_Loot/Containers/Cemetery/Coffin_SearchEnd_Cue.Coffin_SearchEnd_Cue" (Transformation|GetActorLocation)))

(fn ConstructionScript)
